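# Control Rig for the M_MED_Llama_Hero_Winter player body that procedurally adjusts hood bones and defines knee corrective bones.
# From BeginExecution, clamped remaps of head rotation rotate the back hood and shift and rotate the front and front-left hood bones.
# A two-stage knee chain blends deform-bone poses by calf bend, but is not connected to BeginExecution.
# In that chain, the left knee's first stage reads the right calf's angle.
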ControlRigBlueprint /Game/Characters/Player/Male/Medium/Bodies/M_MED_Llama_Hero_Winter/Meshes/M_MED_Llama_Hero_Winter_CtrlRig/M_MED_Llama_Hero_Winter_CtrlRig
graphs [
 RigVMFunctionLibrary {links [], nodes []}
 RigVMModel {nodes [
   n1 = &RigVMFunction_MathQuaternionToEuler(RotationOrder=ZYX, Value=&RigUnit_GetRelativeBoneTransform(Bone=calf_r, Space=thigh_r).Transform.Rotation)
   n2 = &RigVMFunction_MathFloatRemap(SourceMaximum=90, SourceMinimum=4, TargetMinimum=0, bClamp=true, Value=@n1.Result.Z)
   n3 = &RigVMFunction_MathQuaternionToEuler(
    RotationOrder=ZYX
    Value=/Script/ControlRig.RigUnit_GetInitialBoneTransform(Bone=deform_knee_bk_01_l, Space=LocalSpace) {is=struct RigVMUnitNode}.Transform.Rotation
   )
   /Script/ControlRig.RigUnit_ModifyBoneTransforms(
    BoneToModify=((Bone="deform_knee_bk_01_l",Transform=(Rotation=(X=0,Y=0,Z=0,W=1),Translation=(X=-4.319,Y=-4.798,Z=0),Scale3D=(X=1,Y=1,Z=0.7))))
    ExecutePin=()
    Mode=OverrideLocal
    WeightMaximum=1
    Weight=@n2.Result
    BoneToModify.0.Transform.Rotation=&RigVMFunction_MathQuaternionFromEuler(Euler=/Script/RigVM.RigVMFunction_MathVectorAdd(B=(X=0,Y=0,Z=0), A=@n3.Result) {is=struct RigVMUnitNode, template="Add::Execute(in A,in B,out Result)"}.Result).Result
   ) {is=struct RigVMUnitNode}
   n4 = &RigVMFunction_MathQuaternionToEuler(RotationOrder=ZYX, Value=&RigUnit_GetInitialBoneTransform(Bone=deform_knee_fr_01_l, Space=LocalSpace).Transform.Rotation)
   > &RigUnit_ModifyBoneTransforms(
    BoneToModify=((Bone="deform_knee_fr_01_l",Transform=(Rotation=(X=0,Y=0,Z=0,W=1),Translation=(X=0.258,Y=-2,Z=0.009),Scale3D=(X=1,Y=1,Z=1))))
    Mode=OverrideLocal
    WeightMaximum=1
    WeightMinimum=0
    Weight=@n2.Result
    BoneToModify.0.Transform.Rotation=&RigVMFunction_MathQuaternionFromEuler(Euler=&RigVMFunction_MathVectorAdd(B=(X=0,Y=0,Z=-15), A=@n4.Result) {template="Add::Execute(in A,in B,out Result)"}.Result).Result
   )
   n5 = &RigVMFunction_MathFloatRemap(
    SourceMaximum=130
    SourceMinimum=113
    TargetMinimum=0
    bClamp=true
    Value=&RigVMFunction_MathQuaternionToEuler(RotationOrder=ZYX, Value=&RigUnit_GetRelativeBoneTransform(Bone=calf_l, Space=thigh_l).Transform.Rotation).Result.Z
   )
   > &RigUnit_ModifyBoneTransforms(
    BoneToModify=((Bone="deform_knee_bk_01_l",Transform=(Rotation=(X=0,Y=0,Z=0,W=1),Translation=(X=-10.633,Y=-3.561,Z=0),Scale3D=(X=1,Y=1,Z=1))))
    Mode=OverrideLocal
    WeightMaximum=1
    WeightMinimum=0
    Weight=@n5.Result
    BoneToModify.0.Transform.Rotation=&RigVMFunction_MathQuaternionFromEuler(Euler=&RigVMFunction_MathVectorAdd(B=(X=0,Y=0,Z=0), A=@n3.Result) {template="Add::Execute(in A,in B,out Result)"}.Result).Result
   )
   > &RigUnit_ModifyBoneTransforms(
    BoneToModify=((Bone="deform_knee_fr_01_l",Transform=(Rotation=(X=0,Y=0,Z=0,W=1),Translation=(X=0.258,Y=-3.591,Z=0.009),Scale3D=(X=1,Y=1,Z=1))))
    Mode=OverrideLocal
    WeightMaximum=1
    WeightMinimum=0
    Weight=@n5.Result
    BoneToModify.0.Transform.Rotation=&RigVMFunction_MathQuaternionFromEuler(Euler=&RigVMFunction_MathVectorAdd(B=(X=0,Y=0,Z=-51.2), A=@n4.Result) {template="Add::Execute(in A,in B,out Result)"}.Result).Result
   )
   n6 = &RigVMFunction_MathQuaternionToEuler(RotationOrder=ZYX, Value=&RigUnit_GetInitialBoneTransform(Bone=deform_knee_bk_01_r, Space=LocalSpace).Transform.Rotation)
   > &RigUnit_ModifyBoneTransforms(
    BoneToModify=((Bone="deform_knee_bk_01_r",Transform=(Rotation=(X=0,Y=0,Z=0,W=1),Translation=(X=4.319,Y=4.798,Z=0),Scale3D=(X=1,Y=2.022,Z=1.95))))
    Mode=OverrideLocal
    WeightMaximum=1
    Weight=@n2.Result
    BoneToModify.0.Transform.Rotation=&RigVMFunction_MathQuaternionFromEuler(Euler=&RigVMFunction_MathVectorAdd(B=(X=0,Y=0,Z=0), A=@n6.Result) {template="Add::Execute(in A,in B,out Result)"}.Result).Result
   )
   n7 = &RigVMFunction_MathQuaternionToEuler(RotationOrder=ZYX, Value=&RigUnit_GetInitialBoneTransform(Bone=deform_knee_fr_01_r, Space=LocalSpace).Transform.Rotation)
   > &RigUnit_ModifyBoneTransforms(
    BoneToModify=((Bone="deform_knee_fr_01_r",Transform=(Rotation=(X=0,Y=0,Z=0,W=1),Translation=(X=-1.701,Y=-1.197,Z=0),Scale3D=(X=1,Y=1,Z=1))))
    Mode=OverrideLocal
    WeightMaximum=1
    WeightMinimum=0
    Weight=@n2.Result
    BoneToModify.0.Transform.Rotation=&RigVMFunction_MathQuaternionFromEuler(Euler=&RigVMFunction_MathVectorAdd(B=(X=0,Y=0,Z=0), A=@n7.Result) {template="Add::Execute(in A,in B,out Result)"}.Result).Result
   )
   n8 = &RigVMFunction_MathFloatRemap(SourceMaximum=130, SourceMinimum=89, TargetMinimum=0, bClamp=true, Value=@n1.Result.Z)
   > &RigUnit_ModifyBoneTransforms(
    BoneToModify=((Bone="deform_knee_bk_01_r",Transform=(Rotation=(X=0,Y=0,Z=0,W=1),Translation=(X=10.633,Y=3.561,Z=0),Scale3D=(X=1,Y=1,Z=1.777))))
    Mode=OverrideLocal
    WeightMaximum=1
    WeightMinimum=0
    Weight=@n8.Result
    BoneToModify.0.Transform.Rotation=&RigVMFunction_MathQuaternionFromEuler(Euler=&RigVMFunction_MathVectorAdd(B=(X=0,Y=0,Z=0), A=@n6.Result) {template="Add::Execute(in A,in B,out Result)"}.Result).Result
   )
   > &RigUnit_ModifyBoneTransforms(
    BoneToModify=((Bone="deform_knee_fr_01_r",Transform=(Rotation=(X=0,Y=0,Z=0,W=1),Translation=(X=-1.791,Y=-1.35,Z=0),Scale3D=(X=1.913,Y=1,Z=1))))
    Mode=OverrideLocal
    WeightMaximum=1
    WeightMinimum=0
    Weight=@n8.Result
    BoneToModify.0.Transform.Rotation=&RigVMFunction_MathQuaternionFromEuler(Euler=&RigVMFunction_MathVectorAdd(B=(X=0,Y=0,Z=0), A=@n7.Result) {template="Add::Execute(in A,in B,out Result)"}.Result).Result
   )
   RigUnit_BeginExecution = &RigUnit_BeginExecution()
   n9 = &RigVMFunction_MathQuaternionToEuler(
    RotationOrder=ZYX
    Value=&RigUnit_GetRelativeTransformForItem(Child=(Type=Bone,Name="head"), Parent=(Type=Bone,Name="neck_01"), bChildInitial=False, bParentInitial=False).RelativeTransform.Rotation
   )
   &RigUnit_ModifyTransforms(
    ItemToModify=((Item=(Type=Bone,Name="dyn_hood_bk_MSTR"),Transform=(Rotation=(X=0,Y=0,Z=0,W=1),Translation=(X=0,Y=0,Z=0),Scale3D=(X=1,Y=1,Z=1))))
    Weight=1
    WeightMinimum=0
    ItemToModify.0.Transform.Rotation=&RigVMFunction_MathQuaternionFromEuler(
     Euler=(X=0,Y=0,Z=0)
     Euler.Y=&RigVMFunction_MathFloatRemap(SourceMaximum=-55, SourceMinimum=-25, TargetMaximum=5, TargetMinimum=0, bClamp=true, Value=@n9.Result.Z).Result
    ).Result
    ExecutePin=@RigUnit_BeginExecution.ExecutePin
   )
   n10 = &RigVMFunction_MathFloatRemap(
    SourceMaximum=3
    SourceMinimum=-5
    TargetMaximum=-2
    TargetMinimum=0
    bClamp=true
    Value=&RigVMFunction_MathQuaternionToEuler(
     RotationOrder=ZYX
     Value=&RigUnit_GetTransform(Item=(Type=Bone,Name="head"), Space=LocalSpace, bInitial=False).Transform.Rotation
    ).Result.Z
   )
   n11 = &RigVMFunction_MathQuaternionFromEuler(Euler=(X=0,Y=0,Z=0), Euler.Y=@n10.Result)
   > &RigUnit_ModifyTransforms(
    ItemToModify=((Item=(Type=Bone,Name="dyn_hood_fr_base"),Transform=(Rotation=(X=0,Y=0,Z=0,W=1),Translation=(X=0,Y=0,Z=0),Scale3D=(X=1,Y=1,Z=1))))
    Weight=1
    WeightMinimum=0
    ItemToModify.0.Transform.Translation.X=@n10.Result
    ItemToModify.0.Transform.Rotation=@n11.Result
   )
   > &RigUnit_ModifyTransforms(
    ItemToModify=((Item=(Type=Bone,Name="dyn_hood_fr_l_MSTR"),Transform=(Rotation=(X=0,Y=0,Z=0,W=1),Translation=(X=0,Y=0,Z=0),Scale3D=(X=1,Y=1,Z=1))))
    Weight=1
    WeightMinimum=0
    ItemToModify.0.Transform.Translation.X=@n10.Result
    ItemToModify.0.Transform.Rotation=@n11.Result
   )
   RigVMStructNode_1 = &RigVMFunction_MathFloatRemap(SourceMaximum=114, SourceMinimum=4, TargetMinimum=0, bClamp=true)
  ]}
]
hierarchy [
 root {loc=<0 0 0>, rot=<0 0 0 1>}
 pelvis {parent=root, loc=<0 -0.7071 93.1191>, rot=<0.0127 0.707 -0.0127 -0.707>}
 spine_01 {parent=pelvis, loc=<6.7494 -0.6388 0>, rot=<0 0 0.034 -0.9994>}
 spine_02 {parent=spine_01, loc=<7.2518 -0.035 0>, rot=<0 0 0.0238 -0.9997>}
 spine_03 {parent=spine_02, loc=<6.0104 0 0>, rot=<0 0 0.0149 -0.9999>}
 spine_04 {parent=spine_03, loc=<8.1024 0 0>, rot=<0 0 -0.0132 -0.9999>}
 spine_05 {parent=spine_04, loc=<10.0146 0 0>, rot=<0 0 -0.0015 -1>}
 clavicle_l {parent=spine_05, loc=<4.3575 -0.1466 -2.3223>, rot=<0.0465 -0.7289 -0.009 -0.6829>}
 upperarm_l {parent=clavicle_l, loc=<15.5111 0 0>, rot=<0.0226 -0.3234 -0.0088 -0.9459>}
 lowerarm_l {parent=upperarm_l, loc=<26.6514 0 0>, rot=<0.0391 -0.0099 -0.2453 -0.9686>}
 hand_l {parent=lowerarm_l, loc=<22.3491 0 0>, rot=<-0.7081 -0.0507 0.0975 -0.6975>}
 index_metacarpal_l {parent=hand_l, loc=<1.6535 0.0183 -2.96>, rot=<0.0158 -0.0124 -0.0173 -0.9996>}
 index_01_l {parent=index_metacarpal_l, loc=<8.3008 0.0357 -0.0524>, rot=<-0.0044 0.0485 0.0924 -0.9945>}
 index_02_l {parent=index_01_l, loc=<3.3587 -0.3004 0.0452>, rot=<-0.002 0.0027 0.3225 -0.9466>}
 index_03_l {parent=index_02_l, loc=<2.5578 -0.123 -0.0252>, rot=<-0.002 0.0067 0.2558 -0.9667>}
 middle_metacarpal_l {parent=hand_l, loc=<2.1856 -0.1725 -0.5984>, rot=<0.006 0.0168 -0.0329 -0.9993>}
 middle_01_l {parent=middle_metacarpal_l, loc=<7.3275 -0.2447 0.0129>, rot=<-0.0208 0.0042 0.0858 -0.9961>}
 middle_02_l {parent=middle_01_l, loc=<4.0123 0.1401 0.1175>, rot=<-0.0218 0.0258 0.3512 -0.9357>}
 middle_03_l {parent=middle_02_l, loc=<2.6077 0.0534 0.1373>, rot=<0.0082 0.0188 0.1708 -0.9851>}
 pinky_metacarpal_l {parent=hand_l, loc=<1.7616 -0.9753 3.3814>, rot=<-0.0397 0.0814 -0.069 -0.9935>}
 pinky_01_l {parent=pinky_metacarpal_l, loc=<7.1709 -0.4831 -0.1094>, rot=<-0.0654 -0.0035 0.2242 -0.9723>}
 pinky_02_l {parent=pinky_01_l, loc=<2.8274 0.0893 -0.1043>, rot=<0.0037 0.0011 0.2854 -0.9584>}
 pinky_03_l {parent=pinky_02_l, loc=<2.2046 0.0307 -0.0078>, rot=<-0.0002 0.0098 0.1445 -0.9895>}
 ring_metacarpal_l {parent=hand_l, loc=<2.2074 -0.3625 1.3638>, rot=<-0.0284 0.0722 -0.0458 -0.9959>}
 ring_01_l {parent=ring_metacarpal_l, loc=<7.0295 -0.6197 -0.1446>, rot=<-0.0489 -0.0252 0.1892 -0.9804>}
 ring_02_l {parent=ring_01_l, loc=<3.6174 0.4646 -0.1027>, rot=<0 0 0.2806 -0.9598>}
 ring_03_l {parent=ring_02_l, loc=<2.4152 0.1141 0.0199>, rot=<0 0 0.1189 -0.9929>}
 thumb_01_l {parent=hand_l, loc=<1.3118 -0.3191 -2.459>, rot=<0.6546 -0.4042 0.0664 -0.6354>}
 thumb_02_l {parent=thumb_01_l, loc=<5.9336 -0.0448 0.0401>, rot=<-0.0037 0.0213 0.298 -0.9543>}
 thumb_03_l {parent=thumb_02_l, loc=<2.2223 0.3682 -0.0907>, rot=<0.0118 -0.013 0.0066 -0.9998>}
 lowerarm_twist_01_l {parent=lowerarm_l, loc=<17.8549 0 0>, rot=<0 0 0 -1>}
 lowerarm_twist_02_l {parent=lowerarm_l, loc=<9.6199 0 0>, rot=<0 0 0 -1>}
 upperarm_twist_01_l {parent=upperarm_l, loc=<0.0041 0 0>, rot=<0 0 0 -1>}
 upperarm_twist_02_l {parent=upperarm_l, loc=<13.7941 0 0>, rot=<0 0 0 -1>}
 clavicle_r {parent=spine_05, loc=<4.3575 -0.1466 2.3223>, rot=<0.7289 0.0465 -0.6829 0.009>}
 upperarm_r {parent=clavicle_r, loc=<-15.5111 0 0>, rot=<0.0226 -0.3234 -0.0088 -0.9459>}
 lowerarm_r {parent=upperarm_r, loc=<-26.6514 0 0>, rot=<0.0391 -0.0099 -0.2453 -0.9686>}
 hand_r {parent=lowerarm_r, loc=<-22.3492 0 0>, rot=<-0.7081 -0.0507 0.0975 -0.6975>}
 index_metacarpal_r {parent=hand_r, loc=<-1.6534 -0.0181 2.9601>, rot=<0.0158 -0.0124 -0.0173 -0.9996>}
 index_01_r {parent=index_metacarpal_r, loc=<-8.3008 -0.0357 0.0524>, rot=<-0.0044 0.0485 0.0924 -0.9945>}
 index_02_r {parent=index_01_r, loc=<-3.3589 0.3003 -0.0453>, rot=<-0.002 0.0027 0.3225 -0.9466>}
 index_03_r {parent=index_02_r, loc=<-2.5578 0.1231 0.0251>, rot=<-0.002 0.0067 0.2558 -0.9667>}
 middle_metacarpal_r {parent=hand_r, loc=<-2.1856 0.1726 0.5984>, rot=<0.006 0.0168 -0.0329 -0.9993>}
 middle_01_r {parent=middle_metacarpal_r, loc=<-7.3273 0.2447 -0.0129>, rot=<-0.0208 0.0042 0.0858 -0.9961>}
 middle_02_r {parent=middle_01_r, loc=<-4.0142 -0.1411 -0.07>, rot=<-0.0217 0.0259 0.3512 -0.9357>}
 middle_03_r {parent=middle_02_r, loc=<-2.6074 -0.0534 -0.1312>, rot=<0.0083 0.0188 0.1708 -0.9851>}
 pinky_metacarpal_r {parent=hand_r, loc=<-1.7617 0.9753 -3.3814>, rot=<-0.0397 0.0814 -0.069 -0.9935>}
 pinky_01_r {parent=pinky_metacarpal_r, loc=<-7.1707 0.4832 0.1094>, rot=<-0.0654 -0.0035 0.2242 -0.9723>}
 pinky_02_r {parent=pinky_01_r, loc=<-2.8274 -0.0892 0.1043>, rot=<0.0037 0.0011 0.2854 -0.9584>}
 pinky_03_r {parent=pinky_02_r, loc=<-2.2045 -0.031 0.0078>, rot=<-0.0002 0.0098 0.1445 -0.9895>}
 ring_metacarpal_r {parent=hand_r, loc=<-2.2073 0.3627 -1.3638>, rot=<-0.0284 0.0722 -0.0458 -0.9959>}
 ring_01_r {parent=ring_metacarpal_r, loc=<-7.0297 0.6197 0.1446>, rot=<-0.0489 -0.0252 0.1892 -0.9804>}
 ring_02_r {parent=ring_01_r, loc=<-3.6172 -0.4648 0.1027>, rot=<0 0 0.2806 -0.9598>}
 ring_03_r {parent=ring_02_r, loc=<-2.4152 -0.1141 -0.0199>, rot=<0 0 0.1189 -0.9929>}
 thumb_01_r {parent=hand_r, loc=<-1.3119 0.3192 2.459>, rot=<0.6546 -0.4042 0.0664 -0.6354>}
 thumb_02_r {parent=thumb_01_r, loc=<-5.9335 0.0447 -0.0401>, rot=<-0.0037 0.0213 0.298 -0.9543>}
 thumb_03_r {parent=thumb_02_r, loc=<-2.2223 -0.3682 0.0907>, rot=<0.0118 -0.013 0.0066 -0.9998>}
 lowerarm_twist_01_r {parent=lowerarm_r, loc=<-17.855 0 0>, rot=<0 0 0 1>}
 lowerarm_twist_02_r {parent=lowerarm_r, loc=<-9.6199 0 0>, rot=<0 0 0 1>}
 upperarm_twist_01_r {parent=upperarm_r, loc=<-0.0041 0 0>, rot=<0 0 0 1>}
 upperarm_twist_02_r {parent=upperarm_r, loc=<-13.7942 0 0>, rot=<0 0 0 1>}
 neck_01 {parent=spine_05, loc=<11.3138 0.1529 0>, rot=<0 0 -0.2003 -0.9797>}
 neck_02 {parent=neck_01, loc=<4.5627 0 0>, rot=<0 0 0.0504 -0.9987>}
 head {parent=neck_02, loc=<5.3914 0 0>, rot=<0 0 0.115 -0.9934>}
 dyn_l_string_MSTR {parent=spine_05, loc=<4.197 13.0738 -3.7577>, rot=<-0.392 -0.0717 -0.0608 -0.9151>}
 dyn_l_string_01 {parent=dyn_l_string_MSTR, loc=<0.9123 0 0>, rot=<-0.0784 -0.8834 -0.2823 -0.3658>}
 dyn_l_string_02 {parent=dyn_l_string_01, loc=<3.3309 0 0>, rot=<0.1165 -0.1024 -0.073 -0.9852>}
 dyn_l_string_03 {parent=dyn_l_string_02, loc=<3.9715 0 0>, rot=<-0.057 -0.1051 0.0048 -0.9928>}
 dyn_r_string_MSTR {parent=spine_05, loc=<4.0606 13.2685 3.3348>, rot=<0.9865 0.0517 0.1052 0.1148>}
 dyn_r_string_01 {parent=dyn_r_string_MSTR, loc=<0.811 0 0>, rot=<0.0801 -0.8149 0.4804 -0.3142>}
 dyn_r_string_02 {parent=dyn_r_string_01, loc=<2.6877 0 0>, rot=<-0.035 -0.1261 0.0151 -0.9913>}
 dyn_r_string_03 {parent=dyn_r_string_02, loc=<3.663 0 0>, rot=<-0.0421 -0.093 0.027 -0.9944>}
 dyn_hood_bk_MSTR {parent=spine_05, loc=<14.6303 -7.0031 -0.0343>, rot=<0.6817 -0.188 0.188 -0.6817>}
 dyn_hood_bk_base {parent=dyn_hood_bk_MSTR, loc=<0.5859 -0.0343 -2.8863>, rot=<0 0 0 1>}
 dyn_hood_bk {parent=dyn_hood_bk_base, loc=<2.7661 0 0>, rot=<0 -0.5287 0 -0.8488>}
 dyn_hood_rt_MSTR {parent=spine_05, loc=<15.6879 0.371 8.324>, rot=<0.9917 0.0404 0.1224 0.0003>}
 dyn_hood_rt_base {parent=dyn_hood_rt_MSTR, loc=<-1.2872 8.9116 -2.7886>, rot=<0 0 0 -1>}
 dyn_hood_rt {parent=dyn_hood_rt_base, loc=<2.3501 0 0>, rot=<-0.0032 -0.6153 -0.0041 -0.7883>}
 dyn_hood_lt_MSTR {parent=spine_05, loc=<15.8823 -0.2607 -8.5753>, rot=<0.0015 -0.1456 -0.0403 -0.9885>}
 dyn_hood_lt_base {parent=dyn_hood_lt_MSTR, loc=<-1.3722 -8.2562 -2.262>, rot=<0 0 0 1>}
 dyn_hood_lt {parent=dyn_hood_lt_base, loc=<2.273 0 0>, rot=<0.0026 -0.5966 0.0035 -0.8025>}
 dyn_hood_fr_MSTR {parent=spine_05, loc=<7.4869 12.7175 -0.0343>, rot=<-0.6876 -0.2046 -0.2098 -0.6643>}
 dyn_hood_fr_base {parent=dyn_hood_fr_MSTR, loc=<2.0716 0 0>, rot=<0.0155 0.1193 0.0065 -0.9927>}
 dyn_coat_zipper_02_MSTR {parent=dyn_hood_fr_MSTR, loc=<4.4288 -2.4045 1.2996>, rot=<0.3622 -0.3502 0.621 -0.6004>}
 dyn_coat_zipper_02 {parent=dyn_coat_zipper_02_MSTR, loc=<0.4306 0.9807 -0.344>, rot=<0.3961 -0.3762 -0.5257 -0.6521>}
 dyn_hood_fr_l_MSTR {parent=spine_05, loc=<9.78 9.5451 -7.4221>, rot=<-0.3586 -0.1018 -0.2748 -0.8863>}
 dyn_hood_fr_l_base {parent=dyn_hood_fr_l_MSTR, loc=<2.0716 0 0>, rot=<0.0155 0.1193 0.0065 -0.9927>}
 dyn_hood_fr_r_MSTR {parent=spine_05, loc=<9.8209 9.0368 7.7226>, rot=<0.8687 0.2625 0.1302 0.3994>}
 dyn_hood_fr_r_base {parent=dyn_hood_fr_r_MSTR, loc=<2.0716 0 0>, rot=<0.0155 0.1193 0.0065 -0.9927>}
 dyn_zipper_01_MSTR {parent=spine_04, loc=<4.138 12.7995 -7.296>, rot=<0.0294 -0.7065 -0.0294 -0.7065>}
 dyn_zipper_01_base {parent=dyn_zipper_01_MSTR, loc=<0.4428 0.4236 -0.3054>, rot=<0.2803 -0.4297 -0.4482 -0.7321>}
 dyn_zipper_01 {parent=dyn_zipper_01_base, loc=<0.3501 0 0.7804>, rot=<0 0 0 1>}
 dyn_bag_MSTR {parent=spine_04, loc=<-11.4651 2.7231 16.6573>, rot=<0.1439 -0.546 0.0471 -0.824>}
 dyn_bag_BASE {parent=dyn_bag_MSTR, loc=<15.2372 0.7835 2.981>, rot=<-0.2169 -0.0091 0.9761 -0.0131>}
 dyn_bag {parent=dyn_bag_BASE, loc=<0 0 0>, rot=<0 0 0 -1>}
 dyn_bag_zipper_MSTR {parent=dyn_bag_MSTR, loc=<2.9495 -12.0981 -0.5553>, rot=<-0.0608 -0.082 0.9802 -0.1697>}
 dyn_bag_zipper {parent=dyn_bag_zipper_MSTR, loc=<0.7426 0 0>, rot=<0.3625 -0.3933 -0.1665 -0.8284>}
 thigh_l {parent=pelvis, loc=<-2.5693 -0.0109 -7.7501>, rot=<0.0012 0.0572 0.0155 -0.9982>}
 calf_l {parent=thigh_l, loc=<-39.0484 0 0>, rot=<0.0006 -0.0161 -0.0391 -0.9991>}
 calf_twist_01_l {parent=calf_l, loc=<-35.8131 -0.0967 0.0397>, rot=<0.0237 0.0095 -0.0035 -0.9997>}
 calf_twist_02_l {parent=calf_l, loc=<-18.8064 -0.0967 0.0397>, rot=<0.0237 0.0095 -0.0035 -0.9997>}
 foot_l {parent=calf_l, loc=<-42.561 0.1077 0.4292>, rot=<0.0418 -0.0455 0.0343 -0.9975>}
 ball_l {parent=foot_l, loc=<-6.1599 13.8264 0.6806>, rot=<0.0001 -0.0001 -0.7063 -0.7079>}
 deform_knee_bk_01_l {parent=calf_l, loc=<0.5997 -1.0563 -0.0005>, rot=<0 0 -0.0408 -0.9992>}
 deform_knee_fr_01_l {parent=calf_l, loc=<-0.2332 0 0>, rot=<0 0 0 1>}
 thigh_twist_01_l {parent=thigh_l, loc=<-21.6303 0 0>, rot=<0.0013 0 0 -1>}
 dyn_shotgun_shell_01_MSTR {parent=thigh_l, loc=<-0.6056 3.6163 -8.0525>, rot=<0.2975 -0.6567 0.6607 -0.2092>}
 dyn_shotgun_shell_01_base {parent=dyn_shotgun_shell_01_MSTR, loc=<1.2831 0 0>, rot=<-0.2511 -0.2417 0.8491 0.397>}
 dyn_shotgun_shell_01 {parent=dyn_shotgun_shell_01_base, loc=<0 0 0>, rot=<0 0 0 1>}
 dyn_shotgun_shell_02_MSTR {parent=thigh_l, loc=<-0.822 0.997 -8.6641>, rot=<0.3451 -0.6021 0.6853 -0.2206>}
 dyn_shotgun_shell_02_base {parent=dyn_shotgun_shell_02_MSTR, loc=<1.4501 0 0>, rot=<0.2002 0.2847 -0.7756 -0.5266>}
 dyn_shotgun_shell_02 {parent=dyn_shotgun_shell_02_base, loc=<0 0 0>, rot=<0 0 0 -1>}
 dyn_shotgun_shell_03_MSTR {parent=thigh_l, loc=<-0.9605 -1.8611 -8.7187>, rot=<0.4673 -0.4988 0.681 -0.2629>}
 dyn_shotgun_shell_03_base {parent=dyn_shotgun_shell_03_MSTR, loc=<2.0717 0 0>, rot=<-0.1035 -0.1768 0.858 0.4709>}
 dyn_shotgun_shell_03 {parent=dyn_shotgun_shell_03_base, loc=<0 0 0>, rot=<0 0 0 1>}
 dyn_thigh_l {parent=thigh_l, loc=<0 0 0>, rot=<0 0 0 1>}
 dyn_pouch_MSTR {parent=dyn_thigh_l, loc=<-1.0257 7.6186 -3.1505>, rot=<-0.3002 -0.7218 -0.5518 -0.2906>}
 dyn_pouch_01 {parent=dyn_pouch_MSTR, loc=<1.3953 0.2992 -0.3666>, rot=<-0.0652 0.4435 0.308 -0.8392>}
 dyn_pouch {parent=dyn_pouch_01, loc=<0.5206 0 0>, rot=<0 0 0 -1>}
 thigh_r {parent=pelvis, loc=<-2.5691 -0.0109 7.7501>, rot=<0.0572 -0.0012 0.9982 0.0155>}
 calf_r {parent=thigh_r, loc=<39.0487 0 0>, rot=<0.0006 -0.0161 -0.0391 -0.9991>}
 calf_twist_01_r {parent=calf_r, loc=<35.8132 0.0967 -0.0397>, rot=<0.0237 0.0095 -0.0035 -0.9997>}
 calf_twist_02_r {parent=calf_r, loc=<18.8064 0.0967 -0.0397>, rot=<0.0237 0.0095 -0.0035 -0.9997>}
 foot_r {parent=calf_r, loc=<42.5611 -0.1077 -0.4292>, rot=<0.0418 -0.0455 0.0343 -0.9975>}
 ball_r {parent=foot_r, loc=<6.1598 -13.8264 -0.6806>, rot=<0.0001 -0.0001 -0.7063 -0.7079>}
 deform_knee_fr_01_r {parent=calf_r, loc=<0.2341 0 0>, rot=<0 0 0 1>}
 deform_knee_bk_01_r {parent=calf_r, loc=<-0.6022 1.0607 0>, rot=<0 0 -0.0408 -0.9992>}
 thigh_twist_01_r {parent=thigh_r, loc=<21.6303 0 0>, rot=<0.0013 0 0 -1>}
]
