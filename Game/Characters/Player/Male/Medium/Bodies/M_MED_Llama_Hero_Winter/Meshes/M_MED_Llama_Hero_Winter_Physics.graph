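# Physics asset for the medium male Llama Hero Winter body, set up for secondary motion of its dangling zippers, pouch and shotgun shells.
# Capsules on the pelvis and spine form a linked torso that anchors each accessory chain through a MSTR body.
# Free-rotation joints occur along these chains, while tip joints set narrow angle limits for pouch and shells and wider swing limits for zipper pulls.
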PhysicsAsset /Game/Characters/Player/Male/Medium/Bodies/M_MED_Llama_Hero_Winter/Meshes/M_MED_Llama_Hero_Winter_Physics
bodies {
 dyn_bag_zipper [{shapeType=Box, extentX=2.1938, extentY=0.7411, extentZ=2.6349, center=<1.0222 0.006 -0.0053>, rotation=<rot 0 89.5766 -89.8849>}]
 dyn_bag_zipper_MSTR [{shapeType=Box, extentX=1.2536, extentY=8.0896, extentZ=5.7848, center=<0.4005 -0.4513 -1.6025>, rotation=<rot -26.6818 -126.3436 -90.668>}]
 dyn_coat_zipper_02 [{shapeType=Box, extentX=1.9403, extentY=0.9129, extentZ=3.3682, center=<1.3042 -0.0177 -0.0032>, rotation=<rot 0 90.5777 -89.9996>}]
 dyn_coat_zipper_02_MSTR [{shapeType=Box, extentX=1.3015, extentY=6.4397, extentZ=7.885, center=<0.5669 0.096 -1.8312>, rotation=<rot -17.4812 -89.9999 -10.8378>}]
 dyn_pouch [{shapeType=Box, extentX=8.175, extentY=3.9967, extentZ=10.8002, center=<1.3542 0.2231 -0.2819>, rotation=<rot -2.8141 -70.3984 11.287>}]
 dyn_pouch_01 [{shapeType=Capsule, length=0.505, radius=0.505, rotation=<rot 0 -90 0>}]
 dyn_pouch_MSTR [{shapeType=Box, extentX=1.01, extentY=1.01, extentZ=1.01, rotation=<rot 0 -90 -44.9999>}]
 dyn_shotgun_shell_01 [{shapeType=Capsule, length=3.7785, radius=1.8477, center=<1.6977 0.023 0.1985>, rotation=<rot 1.3265 -90 7.2025>}]
 dyn_shotgun_shell_01_MSTR [{shapeType=Box, extentX=1.01, extentY=2.01, extentZ=2.01, rotation=<rot -10 90 0>}]
 dyn_shotgun_shell_01_base [{shapeType=Capsule, length=0.505, radius=0.505, rotation=<rot 0 -90 0>}]
 dyn_shotgun_shell_02 [{shapeType=Capsule, length=3.7731, radius=1.8002, center=<1.7623 0.0349 0.1995>, rotation=<rot -1.3328 -90 6.8086>}]
 dyn_shotgun_shell_02_MSTR [{shapeType=Box, extentX=1.01, extentY=2.01, extentZ=2.01, rotation=<rot -10 89.9999 0>}]
 dyn_shotgun_shell_02_base [{shapeType=Capsule, length=0.505, radius=0.505, rotation=<rot 0 -90 0>}]
 dyn_shotgun_shell_03 [{shapeType=Capsule, length=3.7688, radius=1.7999, center=<1.8026 -0.0235 0.255>, rotation=<rot 0.8168 -90 8.1307>}]
 dyn_shotgun_shell_03_MSTR [{shapeType=Box, extentX=1.01, extentY=2.01, extentZ=2.01, rotation=<rot 0 89.9999 0>}]
 dyn_shotgun_shell_03_base [{shapeType=Capsule, length=0.505, radius=0.505, rotation=<rot 0 -90 0>}]
 dyn_zipper_01 [{shapeType=Box, extentX=1.5782, extentY=0.7956, extentZ=3.3642, center=<1.3035 0.0201 -0.0002>, rotation=<rot 0 90 90>}]
 dyn_zipper_01_MSTR [{shapeType=Box, extentX=1.155, extentY=3.9856, extentZ=1.6949, center=<0.7112 0.8073 -1.7609>, rotation=<rot -23.3989 -111.8795 9.0623>}]
 dyn_zipper_01_base [{shapeType=Box, extentX=11.2283, extentY=1.5224, extentZ=10.5875, center=<1.2345 0.0226 -0.6486>, rotation=<rot -0.9386 -91.3264 -84.6824>}]
 pelvis [{shapeType=Capsule, length=11.099, radius=10.6191, center=<-3.8497 -1.2458 0>}]
 spine_02 [{shapeType=Capsule, length=8.0019, radius=12.3675}]
 spine_03 [{shapeType=Capsule, length=8.4743, radius=12.183}]
 spine_04 [{shapeType=Capsule, length=9.0829, radius=13.4, center=<5.2129 2.3887 0.0001>}]
 spine_05 [{shapeType=Capsule, length=15.7043, radius=13.2778, center=<-3.3777 0.9809 0>}]
}
constraints [
 {bone1Name=spine_02, bone2Name=pelvis, swing1LimitDegrees=45, swing2LimitDegrees=45, twistLimitDegrees=45}
 {bone1Name=spine_03, bone2Name=spine_02, swing1LimitDegrees=45, swing2LimitDegrees=45, twistLimitDegrees=45}
 {bone1Name=spine_04, bone2Name=spine_03, swing1LimitDegrees=45, swing2LimitDegrees=45, twistLimitDegrees=45}
 {bone1Name=spine_05, bone2Name=spine_04, swing1LimitDegrees=45, swing2LimitDegrees=45, twistLimitDegrees=45}
 {bone1Name=dyn_shotgun_shell_03_MSTR, bone2Name=pelvis, swing1LimitDegrees=45, swing2LimitDegrees=45, twistLimitDegrees=15}
 {bone1Name=dyn_shotgun_shell_01_MSTR, bone2Name=pelvis, swing1LimitDegrees=45, swing2LimitDegrees=45, twistLimitDegrees=15}
 {bone1Name=dyn_shotgun_shell_02_MSTR, bone2Name=pelvis, swing1LimitDegrees=45, swing2LimitDegrees=45, twistLimitDegrees=15}
 {
  bone1Name=dyn_pouch_MSTR
  bone2Name=pelvis
  swing1LimitDegrees=45
  swing1Motion=Free
  swing2LimitDegrees=45
  swing2Motion=Free
  twistLimitDegrees=45
  twistMotion=Free
 }
 {
  bone1Name=dyn_coat_zipper_02_MSTR
  bone2Name=spine_05
  swing1LimitDegrees=45
  swing1Motion=Free
  swing2LimitDegrees=45
  swing2Motion=Free
  twistLimitDegrees=45
  twistMotion=Free
 }
 {
  bone1Name=dyn_zipper_01_MSTR
  bone2Name=spine_04
  swing1LimitDegrees=45
  swing1Motion=Free
  swing2LimitDegrees=45
  swing2Motion=Free
  twistLimitDegrees=45
  twistMotion=Free
 }
 {bone1Name=dyn_pouch, bone2Name=dyn_pouch_01, swing1LimitDegrees=7, swing2LimitDegrees=3, twistLimitDegrees=3}
 {bone1Name=dyn_shotgun_shell_03, bone2Name=dyn_shotgun_shell_03_base, swing1LimitDegrees=7, swing2LimitDegrees=3, twistLimitDegrees=3}
 {bone1Name=dyn_shotgun_shell_02, bone2Name=dyn_shotgun_shell_02_base, swing1LimitDegrees=7, swing2LimitDegrees=3, twistLimitDegrees=3}
 {bone1Name=dyn_shotgun_shell_01, bone2Name=dyn_shotgun_shell_01_base, swing1LimitDegrees=7, swing2LimitDegrees=3, twistLimitDegrees=3}
 {bone1Name=dyn_bag_zipper, bone2Name=dyn_bag_zipper_MSTR, swing1LimitDegrees=30, swing2LimitDegrees=120, twistLimitDegrees=3}
 {
  bone1Name=dyn_zipper_01_base
  bone2Name=dyn_zipper_01_MSTR
  swing1LimitDegrees=45
  swing1Motion=Free
  swing2LimitDegrees=45
  swing2Motion=Free
  twistLimitDegrees=45
  twistMotion=Free
 }
 {bone1Name=dyn_zipper_01, bone2Name=dyn_zipper_01_base, swing1LimitDegrees=30, swing2LimitDegrees=120, twistLimitDegrees=3}
 {bone1Name=dyn_coat_zipper_02, bone2Name=dyn_coat_zipper_02_MSTR, swing1LimitDegrees=30, swing2LimitDegrees=120, twistLimitDegrees=15}
 {
  bone1Name=dyn_bag_zipper_MSTR
  bone2Name=spine_04
  swing1LimitDegrees=45
  swing1Motion=Free
  swing2LimitDegrees=45
  swing2Motion=Free
  twistLimitDegrees=45
  twistMotion=Free
 }
 {
  bone1Name=dyn_shotgun_shell_03_base
  bone2Name=dyn_shotgun_shell_03_MSTR
  swing1LimitDegrees=45
  swing1Motion=Free
  swing2LimitDegrees=45
  swing2Motion=Free
  twistLimitDegrees=45
  twistMotion=Free
 }
 {
  bone1Name=dyn_shotgun_shell_02_base
  bone2Name=dyn_shotgun_shell_02_MSTR
  swing1LimitDegrees=45
  swing1Motion=Free
  swing2LimitDegrees=45
  swing2Motion=Free
  twistLimitDegrees=45
  twistMotion=Free
 }
 {
  bone1Name=dyn_shotgun_shell_01_base
  bone2Name=dyn_shotgun_shell_01_MSTR
  swing1LimitDegrees=45
  swing1Motion=Free
  swing2LimitDegrees=45
  swing2Motion=Free
  twistLimitDegrees=45
  twistMotion=Free
 }
 {
  bone1Name=dyn_pouch_01
  bone2Name=dyn_pouch_MSTR
  swing1LimitDegrees=45
  swing1Motion=Free
  swing2LimitDegrees=45
  swing2Motion=Free
  twistLimitDegrees=45
  twistMotion=Free
 }
]
