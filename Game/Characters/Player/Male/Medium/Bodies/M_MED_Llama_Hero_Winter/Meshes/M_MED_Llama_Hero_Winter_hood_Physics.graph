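# Physics asset for the hood mesh of the medium male winter llama hero, covering secondary motion of the hood, two drawstrings and a bag.
# Capsule bodies on spine_04 and spine_05 act as the torso that the dyn_ bones attach to.
# Hood back, left and right bases and both string masters join spine_05, and the bag base joins spine_04, all with free rotation.
# Beyond these roots, tighter angular limits constrain each hood piece, three-link string and bag.
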PhysicsAsset /Game/Characters/Player/Male/Medium/Bodies/M_MED_Llama_Hero_Winter/Meshes/M_MED_Llama_Hero_Winter_hood_Physics
bodies {
 dyn_bag [{shapeType=Capsule, length=12.3198, radius=7.1432, center=<16.5142 1 5>, rotation=<rot -8.4565 8.9765 -76.4063>}]
 dyn_bag_BASE [{
   shapeType=Capsule
   length=3.315
   radius=18.0948
   shapeName=dyn_bag_BASE_capsule
   center=<2.1367 -5.7144 -14.5195>
   rotation=<rot 0.0001 -90.0001 -50>
  }]
 dyn_hood_bk [{shapeType=Capsule, length=9.8411, radius=4.958, center=<0 0 5.4043>, rotation=<rot 0 0 90>}]
 dyn_hood_bk_base [{shapeType=Capsule, length=0.505, radius=0.505, rotation=<rot 0 -90 0>}]
 dyn_hood_lt [{shapeType=Capsule, length=8.0094, radius=4.5859, center=<0 0 4>, rotation=<rot 0 158.4541 90>}]
 dyn_hood_lt_base [{shapeType=Capsule, length=0.505, radius=0.505, rotation=<rot 0 -90 0>}]
 dyn_hood_rt [{shapeType=Capsule, length=9.1367, radius=4.8781, center=<0 0 4>, rotation=<rot 0 23.537 90>}]
 dyn_hood_rt_base [{shapeType=Capsule, length=0.505, radius=0.505, rotation=<rot 0 172.9483 -77.0671>}]
 dyn_l_string_01 [{shapeType=Capsule, length=2, radius=1, center=<2 0 0>, rotation=<rot 90 0 0>}]
 dyn_l_string_02 [{shapeType=Capsule, length=1.9897, radius=0.9738, center=<1.9907 0 0>, rotation=<rot 0 90 90>}]
 dyn_l_string_03 [{shapeType=Capsule, length=1.6865, radius=1.3483, center=<2.4045 0 0>, rotation=<rot 90 0 0>}]
 dyn_l_string_MSTR [{shapeType=Capsule, length=0.9522, radius=0.8721, center=<0.9205 -0.0968 -0.0743>, rotation=<rot 0 88.3378 -85.5488>}]
 dyn_r_string_01 [{shapeType=Capsule, length=1.3717, radius=0.7409, center=<1.5 0 0>, rotation=<rot 90 0 0>}]
 dyn_r_string_02 [{shapeType=Capsule, length=2.0983, radius=0.7996, center=<1.5 0 0>, rotation=<rot 90 0 0>}]
 dyn_r_string_03 [{shapeType=Capsule, length=1.8903, radius=0.98, center=<1.5 0 0>, rotation=<rot 90 0 0>}]
 dyn_r_string_MSTR [{shapeType=Capsule, length=0.8878, radius=0.8173, center=<0.8539 0.1771 0.1753>, rotation=<rot 0 -95.3093 -86.4626>}]
 spine_04 [{shapeType=Capsule, length=9.0829, radius=13.4, center=<5.2129 2.3887 0.0001>}]
 spine_05 [{shapeType=Capsule, length=15.7043, radius=13.2778, center=<-3.3777 0.9809 0>}]
}
constraints [
 {bone1Name=spine_05, bone2Name=spine_04, swing1LimitDegrees=45, swing2LimitDegrees=45, twistLimitDegrees=45}
 {bone1Name=dyn_r_string_03, bone2Name=dyn_r_string_02, swing1LimitDegrees=30, swing2LimitDegrees=30, twistLimitDegrees=8}
 {bone1Name=dyn_r_string_01, bone2Name=dyn_r_string_MSTR, swing1LimitDegrees=15, swing2LimitDegrees=15, twistLimitDegrees=8}
 {bone1Name=dyn_r_string_02, bone2Name=dyn_r_string_01, swing1LimitDegrees=18, swing2LimitDegrees=25, twistLimitDegrees=8}
 {bone1Name=dyn_l_string_03, bone2Name=dyn_l_string_02, swing1LimitDegrees=30, swing2LimitDegrees=30, twistLimitDegrees=8}
 {bone1Name=dyn_l_string_01, bone2Name=dyn_l_string_MSTR, swing1LimitDegrees=30, swing2LimitDegrees=30, twistLimitDegrees=8}
 {bone1Name=dyn_l_string_02, bone2Name=dyn_l_string_01, swing1LimitDegrees=30, swing2LimitDegrees=30, twistLimitDegrees=8}
 {
  bone1Name=dyn_r_string_MSTR
  bone2Name=spine_05
  swing1LimitDegrees=45
  swing1Motion=Free
  swing2LimitDegrees=45
  swing2Motion=Free
  twistLimitDegrees=45
  twistMotion=Free
 }
 {
  bone1Name=dyn_l_string_MSTR
  bone2Name=spine_05
  swing1LimitDegrees=45
  swing1Motion=Free
  swing2LimitDegrees=45
  swing2Motion=Free
  twistLimitDegrees=45
  twistMotion=Free
 }
 {bone1Name=dyn_hood_bk, bone2Name=dyn_hood_bk_base, swing1LimitDegrees=10, swing2LimitDegrees=5, twistLimitDegrees=5}
 {
  bone1Name=dyn_hood_bk_base
  bone2Name=spine_05
  swing1LimitDegrees=45
  swing1Motion=Free
  swing2LimitDegrees=45
  swing2Motion=Free
  twistLimitDegrees=45
  twistMotion=Free
 }
 {bone1Name=dyn_hood_rt, bone2Name=dyn_hood_rt_base, swing1LimitDegrees=10, swing2LimitDegrees=5, twistLimitDegrees=5}
 {bone1Name=dyn_hood_lt, bone2Name=dyn_hood_lt_base, swing1LimitDegrees=10, swing2LimitDegrees=5, twistLimitDegrees=5}
 {
  bone1Name=dyn_hood_lt_base
  bone2Name=spine_05
  swing1LimitDegrees=45
  swing1Motion=Free
  swing2LimitDegrees=45
  swing2Motion=Free
  twistLimitDegrees=45
  twistMotion=Free
 }
 {
  bone1Name=dyn_hood_rt_base
  bone2Name=spine_05
  swing1LimitDegrees=45
  swing1Motion=Free
  swing2LimitDegrees=45
  swing2Motion=Free
  twistLimitDegrees=45
  twistMotion=Free
 }
 {
  bone1Name=dyn_bag_BASE
  bone2Name=spine_04
  swing1LimitDegrees=45
  swing1Motion=Free
  swing2LimitDegrees=45
  swing2Motion=Free
  twistLimitDegrees=45
  twistMotion=Free
 }
 {bone1Name=dyn_bag, bone2Name=dyn_bag_BASE, swing1LimitDegrees=3, swing2LimitDegrees=15, twistLimitDegrees=13}
]
