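# Post-body animation Blueprint for the Llama Hero Winter outfit, copying the attached parent mesh's pose and adding secondary motion.
# Weighted constraints anchor zippers, strings, pouch, shells, bag and hood to body bones.
# While crouching, additive offsets move shells and, scaled by left-thigh rotation from the event graph, the pouch.
# Two RigidBody simulations, one hood-specific, follow, then a bag-zipper constraint and a final Control Rig.
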AnimBlueprint /Game/Characters/Player/Male/Medium/Bodies/M_MED_Llama_Hero_Winter/Meshes/M_MED_Llama_Hero_Winter_AnimBP
target_skeleton=&/M_MED_Llama_Hero_Winter_Skeleton
graphs [
 AnimGraph {nodes [
   n1 = "Constraint - dyn_zipper_01_MSTR (PPPP)"!(
    ComponentPose=Local To Component(LocalPose=Copy Pose From Mesh() {bUseAttachedParent=true})
    Alpha=1
    ConstraintWeights_0=0.0401
    ConstraintWeights_1:float=0.1684
    ConstraintWeights_2:float=0.5632
    ConstraintWeights_3:float=0.2283
   ) -> (Pose) {
    alphaBoolBlend {blendOption=Linear}
    boneToModify {boneName=dyn_zipper_01_MSTR}
    constraintSetup [{targetBone {boneName=spine_02}}, {targetBone {boneName=spine_03}}, {targetBone {boneName=spine_04}}, {targetBone {boneName=spine_05}}]
    constraintWeights [1,1,1,1]
   }
   n2 = "Constraint - dyn_coat_zipper_02_MSTR (PPPPPPP)"!(
    ComponentPose=@n1
    Alpha=1
    ConstraintWeights_0=0.049
    ConstraintWeights_1:float=0.0183
    ConstraintWeights_2:float=0.0449
    ConstraintWeights_3:float=0.2926
    ConstraintWeights_4:float=0.0107
    ConstraintWeights_5:float=0.0005
    ConstraintWeights_6:float=0.584
   ) -> (Pose) {
    alphaBoolBlend {blendOption=Linear}
    boneToModify {boneName=dyn_coat_zipper_02_MSTR}
    constraintSetup [
     {targetBone {boneName=head}}
     {targetBone {boneName=spine_03}}
     {targetBone {boneName=spine_04}}
     {targetBone {boneName=spine_05}}
     {targetBone {boneName=neck_01}}
     {targetBone {boneName=dyn_hood_lt}}
     {targetBone {boneName=dyn_hood_fr_base}}
    ]
    constraintWeights [1,1,1,1,1,1,1]
   }
   n3 = "Constraint - dyn_r_string_MSTR (PPPPPPPP)"!(
    ComponentPose=@n2
    Alpha=1
    ConstraintWeights_0=0.0014
    ConstraintWeights_1:float=0.0723
    ConstraintWeights_2:float=0.1921
    ConstraintWeights_3:float=0.6807
    ConstraintWeights_4:float=0.0066
    ConstraintWeights_5:float=0.012
    ConstraintWeights_6:float=0.0565
    ConstraintWeights_7:float=0.0002
   ) -> (Pose) {
    alphaBoolBlend {blendOption=Linear}
    boneToModify {boneName=dyn_r_string_MSTR}
    constraintSetup [
     {targetBone {boneName=spine_02}}
     {targetBone {boneName=spine_03}}
     {targetBone {boneName=spine_04}}
     {targetBone {boneName=spine_05}}
     {targetBone {boneName=clavicle_r}}
     {targetBone {boneName=neck_01}}
     {targetBone {boneName=dyn_hood_fr_base}}
     {targetBone {boneName=dyn_hood_rt_base}}
    ]
    constraintWeights [1,1,1,1,1,1,1,1]
   }
   n4 = "Constraint - dyn_l_string_MSTR (PPPPPP)"!(
    ComponentPose=@n3
    Alpha=1
    ConstraintWeights_0=0.0041
    ConstraintWeights_1:float=0.0817
    ConstraintWeights_2:float=0.2182
    ConstraintWeights_3:float=0.7385
    ConstraintWeights_4:float=0.0313
    ConstraintWeights_5:float=0.0002
   ) -> (Pose) {
    alphaBoolBlend {blendOption=Linear}
    boneToModify {boneName=dyn_l_string_MSTR}
    constraintSetup [
     {targetBone {boneName=clavicle_l}}
     {targetBone {boneName=spine_03}}
     {targetBone {boneName=spine_04}}
     {targetBone {boneName=spine_05}}
     {targetBone {boneName=dyn_hood_fr_base}}
     {targetBone {boneName=spine_02}}
    ]
    constraintWeights [1,1,1,1,1,1]
   }
   n5 = "Constraint - dyn_pouch_MSTR (PPPP)"!(
    ComponentPose=@n4
    Alpha=1
    ConstraintWeights_0=0.2659
    ConstraintWeights_1:float=0.7119
    ConstraintWeights_2:float=0.0216
    ConstraintWeights_3:float=1
   ) -> (Pose) {
    alphaBoolBlend {blendOption=Linear}
    boneToModify {boneName=dyn_pouch_MSTR}
    constraintSetup [{targetBone {boneName=pelvis}}, {targetBone {boneName=thigh_twist_01_l}}, {targetBone {boneName=spine_01}}, {targetBone {boneName=None}}]
    constraintWeights [1,1,1,1]
   }
   n6 = "Constraint - dyn_shotgun_shell_01_MSTR (PPPP)"!(
    ComponentPose=@n5
    Alpha=1
    ConstraintWeights_0=0.0366
    ConstraintWeights_1:float=0.5449
    ConstraintWeights_2:float=0.4185
    ConstraintWeights_3:float=1
   ) -> (Pose) {
    alphaBoolBlend {blendOption=Linear}
    boneToModify {boneName=dyn_shotgun_shell_01_MSTR}
    constraintSetup [{targetBone {boneName=spine_01}}, {targetBone {boneName=thigh_twist_01_l}}, {targetBone {boneName=pelvis}}, {targetBone {boneName=None}}]
    constraintWeights [1,1,1,1]
   }
   n7 = "Constraint - dyn_shotgun_shell_02_MSTR (PPPP)"!(
    ComponentPose=@n6
    Alpha=1
    ConstraintWeights_0=0.0363
    ConstraintWeights_1:float=0.4805
    ConstraintWeights_2:float=0.4832
    ConstraintWeights_3:float=1
   ) -> (Pose) {
    alphaBoolBlend {blendOption=Linear}
    boneToModify {boneName=dyn_shotgun_shell_02_MSTR}
    constraintSetup [{targetBone {boneName=spine_01}}, {targetBone {boneName=thigh_twist_01_l}}, {targetBone {boneName=pelvis}}, {targetBone {boneName=None}}]
    constraintWeights [1,1,1,1]
   }
   n8 = "Constraint - dyn_shotgun_shell_03_MSTR (PPPP)"!(
    ComponentPose=@n7
    Alpha=1
    ConstraintWeights_0=0.0332
    ConstraintWeights_1:float=0.426
    ConstraintWeights_2:float=0.5408
    ConstraintWeights_3:float=1
   ) -> (Pose) {
    alphaBoolBlend {blendOption=Linear}
    boneToModify {boneName=dyn_shotgun_shell_03_MSTR}
    constraintSetup [{targetBone {boneName=spine_01}}, {targetBone {boneName=thigh_twist_01_l}}, {targetBone {boneName=pelvis}}, {targetBone {boneName=None}}]
    constraintWeights [1,1,1,1]
   }
   n9 = "Constraint - dyn_bag_MSTR (PPPPP)"!(
    ComponentPose=@n8
    Alpha=1
    ConstraintWeights_0=0.2118
    ConstraintWeights_1:float=0.2628
    ConstraintWeights_2:float=0.2455
    ConstraintWeights_3:float=0.1038
    ConstraintWeights_4:float=0.1761
   ) -> (Pose) {
    alphaBoolBlend {blendOption=Linear}
    boneToModify {boneName=dyn_bag_MSTR}
    constraintSetup [
     {targetBone {boneName=spine_01}}
     {targetBone {boneName=spine_02}}
     {targetBone {boneName=spine_03}}
     {targetBone {boneName=spine_04}}
     {targetBone {boneName=pelvis}}
    ]
    constraintWeights [1,1,1,1,1]
   }
   n10 = "Constraint - dyn_hood_bk_MSTR (PPPP)"!(ComponentPose=@n9, Alpha=1, ConstraintWeights_0=1, ConstraintWeights_1:float=1, ConstraintWeights_2:float=1, ConstraintWeights_3:float=1) -> (Pose) {
    alphaBoolBlend {blendOption=Linear}
    boneToModify {boneName=dyn_hood_bk_MSTR}
    constraintSetup [{targetBone {boneName=neck_01}}, {targetBone {boneName=spine_03}}, {targetBone {boneName=head}}, {targetBone {boneName=spine_05}}]
    constraintWeights [1,1,1,1]
   }
   n11 = "Constraint - dyn_hood_lt_MSTR (PPPPPP)"!(
    ComponentPose=@n10
    Alpha=1
    ConstraintWeights_0=0.08
    ConstraintWeights_1:float=0.0042
    ConstraintWeights_2:float=0.0035
    ConstraintWeights_3:float=0.4879
    ConstraintWeights_4:float=0.303
    ConstraintWeights_5:float=0.1203
   ) -> (Pose) {
    alphaBoolBlend {blendOption=Linear}
    boneToModify {boneName=dyn_hood_lt_MSTR}
    constraintSetup [
     {targetBone {boneName=head}}
     {targetBone {boneName=spine_03}}
     {targetBone {boneName=spine_04}}
     {targetBone {boneName=spine_05}}
     {targetBone {boneName=neck_02}}
     {targetBone {boneName=neck_01}}
    ]
    constraintWeights [1,1,1,1,1,1]
   }
   n12 = "Constraint - dyn_hood_rt_MSTR (PPPP)"!(
    ComponentPose=@n11
    Alpha=1
    ConstraintWeights_0=0.2624
    ConstraintWeights_1:float=0.1001
    ConstraintWeights_2:float=0.0762
    ConstraintWeights_3:float=0.5614
   ) -> (Pose) {
    alphaBoolBlend {blendOption=Linear}
    boneToModify {boneName=dyn_hood_rt_MSTR}
    constraintSetup [{targetBone {boneName=neck_02}}, {targetBone {boneName=neck_01}}, {targetBone {boneName=head}}, {targetBone {boneName=spine_05}}]
    constraintWeights [1,1,1,1]
   }
   n13 = "Transform (Modify) Bone - Bone: dyn_pouch_01"(ComponentPose=@n12, Alpha=Get leg rotation_l() -> (leg rotation_l:double), Translation="4,2,1", Rotation="-25,0.000,0.000", Scale="1,1,1") {
    rotationMode=BMM_Additive
    translationMode=BMM_Additive
    alphaBoolBlend {blendInTime=0.1, blendOption=Linear, blendOutTime=0.1}
    boneToModify {boneName=dyn_pouch_01}
    rotation {pitch=-25}
    translation=<4 2 1>
   }
   n14 = Get bIsCrouching()
   n15 = "Transform (Modify) Bone - Bone: dyn_shotgun_shell_01_base"!(ComponentPose=@n13, bAlphaBoolEnabled:bool=@n14, Translation="2,0,1", Rotation="0,0,0", Scale="1,1,1") -> (Pose) {
    alphaInputType=Bool
    translationMode=BMM_Additive
    alphaBoolBlend {blendInTime=0.1, blendOption=Linear, blendOutTime=0.1}
    boneToModify {boneName=dyn_shotgun_shell_01_base}
    translation {x=2, z=1}
   }
   n16 = "Transform (Modify) Bone - Bone: dyn_shotgun_shell_02_base"!(ComponentPose=@n15, bAlphaBoolEnabled:bool=@n14, Translation="1,-1.5,0", Rotation="0,0,0", Scale="1,1,1") -> (Pose) {
    alphaInputType=Bool
    translationMode=BMM_Additive
    alphaBoolBlend {blendInTime=0.1, blendOption=Linear, blendOutTime=0.1}
    boneToModify {boneName=dyn_shotgun_shell_02_base}
    translation=<1 -1.5>
   }
   n17 = "Transform (Modify) Bone - Bone: dyn_shotgun_shell_03_base"!(ComponentPose=@n16, bAlphaBoolEnabled:bool=@n14, Translation="-0.5,-2.5,0", Rotation="0,0,0", Scale="1,1,1") -> (Pose) {
    alphaInputType=Bool
    translationMode=BMM_Additive
    alphaBoolBlend {blendInTime=0.1, blendOption=Linear, blendOutTime=0.1}
    boneToModify {boneName=dyn_shotgun_shell_03_base}
    translation=<-0.5 -2.5>
   }
   n18 = RigidBody(ComponentPose=@n17) {
    bClampLinearTranslationLimitToRefPose=true
    bOverrideWorldGravity=true
    overridePhysicsAsset=&/M_MED_Llama_Hero_Winter_Physics
    simulationSpace=BaseBoneSpace
    alphaBoolBlend {blendOption=Linear}
    baseBoneRef {boneName=root}
    componentAppliedLinearAccClamp=<3000 3000 3000>
    componentLinearAccScale=<0.3 0.3 0.5>
    componentLinearVelScale=<0.1 0.1 0.2>
    overrideWorldGravity {z=-980}
   }
   n19 = RigidBody(ComponentPose=@n18) {
    bClampLinearTranslationLimitToRefPose=true
    bOverrideWorldGravity=true
    overridePhysicsAsset=&/M_MED_Llama_Hero_Winter_hood_Physics
    simulationSpace=BaseBoneSpace
    alphaBoolBlend {blendOption=Linear}
    baseBoneRef {boneName=spine_02}
    componentAppliedLinearAccClamp=<3000 3000 3000>
    componentLinearAccScale=<0.3 0.3 0.5>
    componentLinearVelScale=<0.1 0.1 0.2>
    overrideWorldGravity {z=-980}
   }
   n20 = "Constraint - dyn_bag_zipper_MSTR (PPPPPPP)"!(
    ComponentPose=@n19
    Alpha=1
    ConstraintWeights_0=0.0591
    ConstraintWeights_1:float=0.1141
    ConstraintWeights_2:float=0.1325
    ConstraintWeights_3:float=0.0779
    ConstraintWeights_4:float=0.0143
    ConstraintWeights_5:float=0.5715
    ConstraintWeights_6:float=0.0306
   ) -> (Pose) {
    alphaBoolBlend {blendOption=Linear}
    boneToModify {boneName=dyn_bag_zipper_MSTR}
    constraintSetup [
     {targetBone {boneName=spine_01}}
     {targetBone {boneName=spine_02}}
     {targetBone {boneName=spine_03}}
     {targetBone {boneName=spine_04}}
     {targetBone {boneName=spine_05}}
     {targetBone {boneName=dyn_bag}}
     {targetBone {boneName=pelvis}}
    ]
    constraintWeights [1,1,1,1,1,1,1]
   }
   n21 = Control Rig(Source=Component To Local(ComponentPose=@n20)) {
    alphaBoolBlend {blendOption=Linear}
    controlRigAssetReference {blueprintRigClass=/Game/Characters/Player/Male/Medium/Bodies/M_MED_Llama_Hero_Winter/Meshes/M_MED_Llama_Hero_Winter_CtrlRig/M_MED_Llama_Hero_Winter_CtrlRig.M_MED_Llama_Hero_Winter_CtrlRig_C}
   }
   Output Pose(Result=@n21) {layerGroup=None}
  ]}
 EventGraph {nodes [
   Event BlueprintUpdateAnimation()
   > n1 = Sequence()
   > n2 = Branch(Condition=Get bIsCrouching())
   > Set crouch float(crouch float:double=1) -> (Output_Get:double)
   Set crouch float(execute=@n2.else, crouch float=0)
   n3 = Branch(execute=@n1.then_1, Condition=Get bIsCrouching())
   n4 = BreakTransform(InTransform=GetDeltaTransformFromRefPose(self=GetOwnerSkeletalMeshComponent(), BoneName=thigh_twist_01_l, BaseName=pelvis))
   > Set leg rotation_l(leg rotation_l:double=MapRangeClamped(Value=BreakRotator(InRot=@n4.Rotation), InRangeA=10, InRangeB=4)) -> (Output_Get:double)
   Set leg rotation_l(execute=@n3.else, leg rotation_l=0)
  ]}
]
variables [
 crouch float {type=number}
 leg rotation_l {type=number}
]
